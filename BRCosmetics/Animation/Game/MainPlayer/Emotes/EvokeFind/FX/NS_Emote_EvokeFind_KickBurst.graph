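# Niagara one-shot kick burst effect for the EvokeFind emote, a blue, cyan and purple impact flash.
# Instantaneous-burst emitters layer timeline-animated mesh audio waves, a circle, streaks, spikes, gravity-driven sparks and a decal-named sprite, with lights on the flash and decal layers.
# Flash_Elements packs glow, star and flare looks into shared particles via per-element attributes and normalized ages feeding multiple renderers.
# Front-end flags and time-of-day multipliers adjust several sizes, offsets and colors.
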NiagaraSystem /BRCosmetics/Animation/Game/MainPlayer/Emotes/EvokeFind/FX/NS_Emote_EvokeFind_KickBurst
dependencies {
 usedDataInterfaces [&NiagaraDataInterfaceCurve, &NiagaraDataInterfaceVector2DCurve, &NiagaraDataInterfaceColorCurve]
 usedDynamicInputs [&Add_Vector, /Niagara/DynamicInputs/Helpers/SimulationPosition]
 usedModules [&SystemState]
}
emitters {
 AudioWaves_Burst {
  emitterSpawnScript [/SetVariables(Emitter.float:NiagaraFloat=&RandomRangeFloat(Minimum=0, Maximum=500)) {isSetParameters=true}]
  emitterUpdateScript [
   &EmitterState()
   &SpawnBurst_Instantaneous(Spawn Count=5, Use Spawn Probability=false, UseLoopCountLimit=false)
  ]
  particleSpawnScript [
   &InitializeParticle(
    Color=&MultiplyLinearColor(
     Linear Color=<rgba 0.25 1.8333 5 10>
     Scale Factor=&MakeVector4_FromLinearColor(LinearColor=$System.TODColorMultiplier)
    )
    Lifetime=&FloatFromCurve(
     FloatCurve=DataInterface() {propertyValues {Curve {keys [<at 0 0.3>,<at 1 0.1>]}}}
     CurveIndex=&ReturnNormalizedExecIndex()
     Scale Curve=2
    )
    Mesh Scale Mode=::NewEnumerator3
    Mesh Uniform Scale=&Multiply_Float(
     A=&Multiply_Float(A=0.22, B=&FN_FrontEndFloatModifier(Gameplay=1, Front End=0.6))
     B=&FloatFromCurve(
      FloatCurve=DataInterface() {propertyValues {Curve {keys [
          {interpMode=RCIM_Cubic, leaveTangent=0.7118, tangentMode=RCTM_Break, time=0, value=0.4}
          {arriveTangent=0.486, tangentMode=RCTM_User, time=1, value=1}
         ]}}}
      CurveIndex=&ReturnNormalizedExecIndex()
     )
    )
    Sprite Size Mode=::NewEnumerator3
    UsePositionOffset=true
    Write Material Random=true
   )
   &Color(Color=&MultiplyLinearColor(
     Linear Color=$Particles.Color
     Scale Factor=&MakeVector4(
      X=1
      Y=1
      Z=1
      W=&FloatFromCurve(
       FloatCurve=DataInterface() {propertyValues {Curve {keys [
           {interpMode=RCIM_Cubic, leaveTangent=0.9764, tangentMode=RCTM_Break, time=0, value=0.2}
           {arriveTangent=0.4606, interpMode=RCIM_Cubic, tangentMode=RCTM_Break, time=1, value=1}
          ]}}}
       CurveIndex=&ReturnNormalizedExecIndex()
      )
     )
    ))
  ]
  particleUpdateScript [
   &ParticleState()
   /SetVariables(Particles.MeshOrientation=&FindQuatBetweenTwoVectors(From Vector=<1 0 0>, To Vector=<0 0 1>)) {isSetParameters=true}
   &Timeline(Initial Start Time:NiagaraFloat=$Emitter.float, Play Rate=0.7, Timeline End=100000)
   &DynamicMaterialParameters(
    Index 0 Param 1=&MakeFloatFromVector(Vector=$Particles.Scale, Channel=::NewEnumerator0)
    Index 0 Param 2=2
    Index 0 Param 3=-0.3
    Index 0 Param 4=&RandomRangeFloat(Minimum=3, Maximum=5)
    Index 1 Param 1=$System.Timeline.Time
    Index 1 Param 2=3
    Index 1 Param 3=0.5
    Index 1 Param 4=$Particles.Timeline.Time
    Index 2 Param 1=2
    Index 2 Param 2=2
    Index 2 Param 3=&FloatFromCurve(
     FloatCurve=DataInterface() {propertyValues {Curve {keys [{interpMode=RCIM_Cubic, leaveTangent=0.3052, tangentMode=RCTM_Break, time=0.35, value=1}, <at 0.5 0>]}}}
     Scale Curve=8
    )
    Module.Index0Float0Write=true
    Module.Index0Float1Write=true
    Module.Index0Float2Write=true
    Module.Index0Float3Write=true
    Module.Index1Float0Write=true
    Module.Index1Float1Write=true
    Module.Index1Float2Write=true
    Module.Index1Float3Write=true
    Module.Index2Float0Write=true
    Module.Index2Float1Write=true
    Module.Index2Float2Write=true
    Module.Index2Float3Write=true
    Write Parameter Index 1=::NewEnumerator1
    Write Parameter Index 2=::NewEnumerator1
   )
   &ScaleMeshSize(Scale Factor=&VectorFromFloat(Value=&FloatFromCurve(
      FloatCurve=DataInterface() {propertyValues {Curve {keys [
          {interpMode=RCIM_Cubic, leaveTangent=1.45, tangentMode=RCTM_Break, time=0, value=0.5}
          {arriveTangent=-0.7179, interpMode=RCIM_Cubic, tangentMode=RCTM_Break, time=1, value=2.5}
         ]}}}
      Scale Curve=&FloatFromCurve(
       FloatCurve=DataInterface() {propertyValues {Curve {keys [<at 0 1.5>,<at 1 1>]}}}
       CurveIndex=&ReturnNormalizedExecIndex()
      )
     )))
   &ScaleColor(
    Scale Alpha=&FloatFromCurve(
     FloatCurve=DataInterface() {propertyValues {Curve {keys [
         {interpMode=RCIM_Cubic, tangentMode=RCTM_User, time=0, value=0}
         {arriveTangent=-3.9775, interpMode=RCIM_Cubic, leaveTangent=-3.9775, tangentMode=RCTM_User, time=0.2, value=1}
         {tangentMode=RCTM_User, time=1, value=0}
        ]}}}
     Scale Curve=0.2
    )
    Scale RGB=&VectorFromFloat(Value=&FloatFromCurve(FloatCurve=DataInterface() {propertyValues {Curve {keys [{tangentMode=RCTM_User, time=0, value=1}, {tangentMode=RCTM_User, time=1, value=1}]}}}))
    ScaleA=true
    ScaleRGB=true
   )
   &ScaleColor(
    Scale Alpha=0.5
    Scale RGB=&VectorFromFloat(Value=&FNTimeOfDayScalar(Night Scale=1, Morning Scale=1, Evening Scale=1))
    ScaleA=true
    ScaleRGB=true
   ) {module=ScaleColor001}
   &CameraOffset(Camera Offset Amount=15)
  ]
  rendererClasses [&NiagaraMeshRendererProperties]
  renderers [{rendererClass=&NiagaraMeshRendererProperties, rendererIndex=0}]
 }
 CircleShape {
  emitterUpdateScript [
   &EmitterState()
   &SpawnBurst_Instantaneous(Use Spawn Probability=false, UseLoopCountLimit=false)
  ]
  particleSpawnScript [
   &InitializeParticle(
    Color Mode=::NewEnumerator0
    Lifetime=0.15
    Sprite Rotation Mode=::NewEnumerator2
    Sprite Size Mode=::NewEnumerator3
    Uniform Sprite Size=100
    UsePositionOffset=false
    Write Material Random=true
   )
   &SpriteFacingAndAlignment(WriteAlignment=false, WriteFacing=true)
  ]
  particleUpdateScript [
   &ParticleState()
   &ScaleSpriteSize(Uniform Curve Sprite Scale=DataInterface() {propertyValues {Curve {keys [{tangentMode=RCTM_User, time=0, value=0}, <at 0.2687 0.8593>, {tangentMode=RCTM_User, time=1.0022, value=1.2275}]}}})
   &SpriteRotationRate(Rotation Rate=&RandomRangeFloat(Minimum=-200, Maximum=200)) {enabled=false}
   &DynamicMaterialParameters(
    Index 0 Param 1=&FloatFromCurve(FloatCurve=DataInterface() {propertyValues {Curve {defaultValue=1, keys [{tangentMode=RCTM_User, time=0.0054, value=0.295}, <at 0.3907 0.9465>, {tangentMode=RCTM_User, time=1, value=1}]}}})
    Module.Index0Float0Write=true
    Module.Index0Float1Write=true
    Module.Index0Float2Write=true
    Module.Index0Float3Write=true
   )
  ]
  rendererClasses [&NiagaraSpriteRendererProperties]
  renderers [{rendererClass=&NiagaraSpriteRendererProperties}]
 }
 Flash_Elements {
  emitterUpdateScript [
   &EmitterState()
   &SpawnBurst_Instantaneous(Spawn Group=1, Use Spawn Probability=false, UseLoopCountLimit=false) {module=SpawnBurst_Instantaneous001}
   &SpawnBurst_Instantaneous(Use Spawn Probability=false, UseLoopCountLimit=false)
  ]
  particleSpawnScript [
   &InitializeParticle(Color Mode=::NewEnumerator0, Lifetime=0.25, UsePositionOffset=false, Write Material Random=true)
   &ShapeLocation(MaskSpawnGroup=false, Override Seed=false, Sphere Radius=1, UseOwnerScale=true)
   &SpriteFacingAndAlignment(Sprite Alignment=$Particles.ShapeLocation.ShapeNormal, WriteAlignment=true, WriteFacing=false)
   &FN_PerParticleLOD(Enable SpriteSize Distance Scaling=true, Sprite Size Scale Far=<2.5 2.5>)
   /SetVariables(Particles.SpriteSize=<1 1>, Particles.VisibilityTag=&SpawnGroup()) {isSetParameters=true}
   /SetVariables(
    Particles.Dark Glow Color:LinearColor=&MultiplyLinearColor(
     Linear Color=<rgba 1 1 1 0.2>
     Scale Factor=&MakeVector4_FromLinearColor(LinearColor=$System.TODColorMultiplier)
    )
    Particles.Dark Glow Lifetime=0.25
    Particles.Dark Glow Sprite Size:Vector2f=&Vector2DFromFloat(Value=&FN_FrontEndFloatModifier(Gameplay=200, Front End=200))
   ) {isSetParameters=true}
   /SetVariables(
    Particles.Flash Star Cam Offset=25
    Particles.Flash Star Color:LinearColor=&MultiplyLinearColor(
     Linear Color=<rgba 1 1 1 1>
     Scale Factor=&MakeVector4_FromLinearColor(LinearColor=$System.TODColorMultiplier)
    )
    Particles.Flash Star Lifetime=0.17
    Particles.Flash Star Rotation:NiagaraFloat=&RandomRangeFloat(Minimum=0, Maximum=360)
    Particles.Flash Star Sprite Size:Vector2f=&Multiply_Vector2D_ByFloat(
     Float=&Multiply_Float(
      A=&MakeCustomFloatFromBool(Bool=$User.bIsFrontEnd, True Float=0.45, False Float=1)
      B=&RandomRangeFloat(Minimum=120, Maximum=150)
     )
     Vector2D=$Particles.SpriteSize
    )
    Particles.Flash Star Sprite Size Alt:Vector2f=&Multiply_Vector2D_ByFloat(
     Float=&Multiply_Float(
      A=0.85
      B=&Multiply_Float(
       A=&MakeCustomFloatFromBool(Bool=$User.bIsFrontEnd, True Float=0.45, False Float=1)
       B=&RandomRangeFloat(Minimum=120, Maximum=150)
      )
     )
     Vector2D=$Particles.SpriteSize
    )
    Particles.Flash Star SubUV Index:NiagaraFloat=&RandomRangeFloat(Minimum=0, Maximum=3)
    Particles.Flash Star SubUV Index Alt:NiagaraFloat=&RandomRangeFloat(Minimum=0, Maximum=3)
   ) {isSetParameters=true}
   /SetVariables(
    Particles.Flash Flare Color:LinearColor=&MultiplyLinearColor(
     Linear Color=<rgba 0 4.5 6 0.7>
     Scale Factor=&MakeVector4_FromLinearColor(LinearColor=$System.TODColorMultiplier)
    )
    Particles.Flash Flare Lifetime:NiagaraFloat=&FN_FrontEndFloatModifier(Gameplay=0.2, Front End=0.125)
    Particles.Flash Flare Sprite Size:Vector2f=&FN_FrontEndVector2Modifier(Front End:Vector2f=<90 90>, Gameplay:Vector2f=<130 130>)
   ) {isSetParameters=true}
   /SetVariables(Particles.Dark_Glow_CameraOffset=&MakeCustomFloatFromBool(Bool=$User.bIsFrontEndPreview, True Float=1, False Float=30)) {isSetParameters=true}
   &DynamicMaterialParameters(
    Index 0 Param 1=0.2
    Index 0 Param 2=1
    Module.Index0Float0Write=true
    Module.Index0Float1Write=true
    Module.Index0Float2Write=true
    Module.Index0Float3Write=true
   )
  ]
  particleUpdateScript [
   &ParticleState()
   /SetVariables(
    Particles.Update Dark Glow Age:NiagaraFloat=&Divide_Float(A=$Particles.Age, B=$Particles.Initial.Dark Glow Lifetime)
    Particles.Update Flash Flare Age:NiagaraFloat=&Divide_Float(A=$Particles.Age, B=$Particles.Initial.Flash Flare Lifetime)
    Particles.Update Flash Star Age:NiagaraFloat=&Divide_Float(A=$Particles.Age, B=$Particles.Initial.Flash Star Lifetime)
   ) {isSetParameters=true}
   /SetVariables(
    Particles.Update Dark Glow Color:LinearColor=&MakeLinearColorFromVectorAndFloat(
     "Vector (RGB)"=&MakeVectorFromLinearColorRGB(Color=&MultiplyLinearColor(
       Linear Color=$Particles.Initial.Dark Glow Color
       Scale Factor=&MakeVector4_FromLinearColor(LinearColor=&Multiply_LinearColorByFloat(
         Linear Color=&ColorFromCurve(
          ColorCurve=DataInterface() {propertyValues {
            AlphaCurve {keys []}
            BlueCurve {keys [<at 0.1777 0.45 RCIM_Cubic>,<at 0.5555 0.6354>,<at 0.7486 0.95>,<at 1 0.8594 RCIM_Cubic>]}
            GreenCurve {keys [<at 0.1777 0.2475 RCIM_Cubic>,<at 0.5555 0.0318>,<at 0.7486 0.095>,<at 1 0.4082 RCIM_Cubic>]}
            RedCurve {keys [<at 0.1777 0.045 RCIM_Cubic>,<at 0.5555 0.233>,<at 0.7486 0.38>,<at 1 0.0859 RCIM_Cubic>]}
           }}
          CurveIndex=$Particles.Update Dark Glow Age
         )
         Float=0.5
        ))
      ))
     "Float (Alpha)"=&FloatFromCurve(
      FloatCurve=DataInterface() {propertyValues {Curve {
         defaultValue=1
         keys [
          {interpMode=RCIM_Cubic, tangentMode=RCTM_User, time=0, value=1}
          {arriveTangent=-1.9502, interpMode=RCIM_Cubic, leaveTangent=-1.9502, tangentMode=RCTM_User, time=1, value=0}
         ]
        }}}
      CurveIndex=$Particles.Update Dark Glow Age
     )
    )
    Particles.Update Dark Glow Sprite Size:Vector2f=&Multiply_Vector2D_ByFloat(
     Float=&FloatFromCurve(FloatCurve=DataInterface() {propertyValues {Curve {keys [<at 0 1 RCIM_Cubic>,<at 1 0.75>]}}}, CurveIndex=$Particles.Update Dark Glow Age)
     Vector2D=$Particles.Initial.Dark Glow Sprite Size
    )
    Particles.Update Flash Star Color:LinearColor=&ColorFromCurve(
     ColorCurve=DataInterface() {propertyValues {
       AlphaCurve {keys [<at 0 1>,<at 0.8992 0>]}
       BlueCurve {keys [<at 0.1978 2>,<at 0.4068 1.5>,<at 0.6127 0.7>]}
       GreenCurve {keys [<at 0.1978 0.4667>,<at 0.4068 0.3243>,<at 0.6127 0.0145>]}
       RedCurve {keys [<at 0.1978 0>,<at 0.4068 0.0891>,<at 0.6127 0.1287>]}
      }}
     CurveIndex=$Particles.Update Flash Star Age
    )
    Particles.Update Flash Star Color Alt:LinearColor=&ColorFromCurve(
     ColorCurve=DataInterface() {propertyValues {
       AlphaCurve {keys [<at 0 1>,<at 0.8992 0>]}
       BlueCurve {keys [<at 0.1978 3>,<at 0.4068 2>,<at 0.6127 1.5>]}
       GreenCurve {keys [<at 0.1978 2.5>,<at 0.4068 1.5297>,<at 0.6127 1.0103>]}
       RedCurve {keys [<at 0.1978 0>,<at 0.4068 0.1188>,<at 0.6127 0.031>]}
      }}
     CurveIndex=$Particles.Update Flash Star Age
    )
    Particles.Update Flash Star Sprize Size:Vector2f=&Multiply_Vector2D_ByFloat(
     Float=&FloatFromCurve(
      FloatCurve=DataInterface() {propertyValues {Curve {keys [
          {arriveTangent=2.7013, interpMode=RCIM_Cubic, leaveTangent=2.7013, tangentMode=RCTM_User, time=0, value=0}
          {interpMode=RCIM_Cubic, tangentMode=RCTM_User, time=1, value=1}
         ]}}}
      CurveIndex=$Particles.Update Flash Star Age
     )
     Vector2D=$Particles.Initial.Flash Star Sprite Size
    )
    Particles.Update Flash Star Sprize Size Alt:Vector2f=&Multiply_Vector2D_ByFloat(
     Float=&FloatFromCurve(
      FloatCurve=DataInterface() {propertyValues {Curve {keys [
          {arriveTangent=2.7013, interpMode=RCIM_Cubic, leaveTangent=2.7013, tangentMode=RCTM_User, time=0, value=0}
          {interpMode=RCIM_Cubic, tangentMode=RCTM_User, time=1, value=1}
         ]}}}
      CurveIndex=$Particles.Update Flash Star Age
     )
     Vector2D=$Particles.Initial.Flash Star Sprite Size Alt
    )
   ) {isSetParameters=true}
   /SetVariables(Particles.Update Flash Flare Color:LinearColor=&MakeLinearColorFromVectorAndFloat(
     "Vector (RGB)"=&Multiply_Vector(
      A=&MakeVectorFromLinearColorRGB(Color=$Particles.Initial.Flash Flare Color)
      B=&VectorFromFloat(Value=&FN_FrontEndFloatModifier(Gameplay=1, Front End=1))
     )
     "Float (Alpha)"=&Multiply_Float(
      A=&MakeFloatFromLinearColor(LinearColor=<rgba 1 1 1 1>, Channel:ENiagaraFloatFromLinearColorOptions=::NewEnumerator0)
      B=&Lerp_Float(A=1, B=0, Alpha=$Particles.Update Flash Flare Age)
     )
    )) {isSetParameters=true}
   &FN_Light_Attributes(
    Color=&MultiplyLinearColor(
     Linear Color=<rgba 0 0.6667 2 1>
     Scale Factor=&MakeVector4_FromLinearColor(LinearColor=$System.TODColorMultiplier)
    )
    Create Color Attribute=true
    Create Exponent Attribute=true
    Create Position Attribute=true
    Create Radius Attribute=true
    Exponent=5
    Position=<15 0 0>
    Radius=90
    Scale Color=&FloatFromCurve(
     FloatCurve=DataInterface() {propertyValues {Curve {keys [
         {interpMode=RCIM_Cubic, tangentMode=RCTM_User, time=0, value=1}
         {arriveTangent=-0.5426, interpMode=RCIM_Cubic, leaveTangent=-0.5426, tangentMode=RCTM_User, time=1, value=0}
        ]}}}
     Scale Curve=10
    )
   )
  ]
  rendererClasses [&NiagaraSpriteRendererProperties, &NiagaraLightRendererProperties]
  renderers [
   {rendererClass=&NiagaraSpriteRendererProperties}
   {rendererClass=&NiagaraSpriteRendererProperties, rendererIndex=1}
   {rendererClass=&NiagaraSpriteRendererProperties, rendererIndex=2}
   {rendererClass=&NiagaraLightRendererProperties, rendererIndex=3}
   {rendererClass=&NiagaraSpriteRendererProperties, rendererIndex=4}
  ]
 }
 Minimal {
  emitterUpdateScript [
   &EmitterState()
   &SpawnBurst_Instantaneous(Spawn Count=5, Use Spawn Probability=false, UseLoopCountLimit=false)
  ]
  particleSpawnScript [
   &InitializeParticle(
    Color Maximum=<rgba 0.7573 0.5273 1 1>
    Color Minimum=<rgba 0.5229 0.9044 1 1>
    Color Mode=::NewEnumerator2
    Lifetime Max=0.2
    Lifetime Min=0.12
    Lifetime Mode=::NewEnumerator1
    Sprite Size Max=<12 85>
    Sprite Size Min=<5 50>
    Sprite Size Mode=::NewEnumerator2
    UsePositionOffset=false
    Write Material Random=true
   )
   &AddVelocity(Module.Use Falloff Distance=false, Override Seed=false, Velocity Mode=::NewEnumerator1, Velocity Speed=1)
   &ShapeLocation(MaskSpawnGroup=false, Override Seed=false, Sphere Radius=5, UseOwnerScale=true) {enabled=false}
  ]
  particleUpdateScript [
   &ParticleState()
   &ScaleSpriteSize(Uniform Curve Sprite Scale=DataInterface() {propertyValues {Curve {keys [{tangentMode=RCTM_User, time=0, value=0}, <at 0.2071 0.7505>, {tangentMode=RCTM_User, time=0.7379, value=1.0157}, <at 1.0102 0.5765>]}}})
  ]
  rendererClasses [&NiagaraSpriteRendererProperties]
  renderers [{rendererClass=&NiagaraSpriteRendererProperties}]
 }
 ParticleDecal {
  emitterUpdateScript [
   &EmitterState()
   &SpawnBurst_Instantaneous(Use Spawn Probability=false, UseLoopCountLimit=false)
  ]
  particleSpawnScript [
   &InitializeParticle(
    Color=<rgba 6.5694 7.2182 8 1>
    Lifetime=0.3
    Mesh Scale Mode=::NewEnumerator3
    Mesh Uniform Scale=2
    Position Offset=<0 0 -20>
    Sprite Size Mode=::NewEnumerator3
    Uniform Sprite Size=100
    UsePositionOffset=true
    Write Material Random=true
   )
   &CameraOffset(Camera Offset Amount=30)
   &DynamicMaterialParameters(
    Index 0 Param 1=1
    Index 0 Param 3=1
    Module.Index0Param0Write=true
    Module.Index0Param1Write=true
    Module.Index0Param2Write=true
    Module.Index0Param3Write=true
   )
  ]
  particleUpdateScript [
   &ParticleState()
   &SkeletalMeshLocation(Bone Sampling Mode=::NewEnumerator4, MaskSpawnGroup=false, Skeletal Mesh=$System.Skeletal Mesh) {enabled=false}
   &ScaleColor(
    Linear Color Curve=DataInterface() {propertyValues {
      AlphaCurve {keys [<at 0.0418 1 RCIM_Cubic>,<at 0.9985 0>]}
      BlueCurve {keys [<at 0 1 RCIM_Cubic>,<at 0.274 1>,<at 0.7128 1 RCIM_Cubic>]}
      GreenCurve {keys [<at 0 0.8879 RCIM_Cubic>,<at 0.274 0.1812>,<at 0.7128 0.0953 RCIM_Cubic>]}
      RedCurve {keys [<at 0 0.159 RCIM_Cubic>,<at 0.274 0.0545>,<at 0.7128 0.9131 RCIM_Cubic>]}
     }}
    Scale Mode=::NewEnumerator2
   )
   &FN_Light_Attributes(
    Color=$Particles.Color
    Create Color Attribute=true
    Create Exponent Attribute=true
    Create Radius Attribute=true
    Exponent=4
    Radius=60
    Scale Color=5
   )
  ]
  rendererClasses [&NiagaraSpriteRendererProperties, &NiagaraLightRendererProperties]
  renderers [
   {rendererClass=&NiagaraSpriteRendererProperties}
   {rendererClass=&NiagaraLightRendererProperties, rendererIndex=1}
  ]
 }
 Sparks {
  emitterUpdateScript [
   &EmitterState()
   &SpawnBurst_Instantaneous(
    Spawn Count=&RandomRangeInt(Minimum=8, Maximum=15)
    Use Spawn Probability=false
    UseLoopCountLimit=false
   )
  ]
  particleSpawnScript [
   &InitializeParticle(
    Color Maximum=<rgba 0.1092 0.8609 3 1>
    Color Minimum=<rgba 1.3497 0.2633 2 1>
    Color Mode=::NewEnumerator2
    Lifetime Max=0.5
    Lifetime Min=0.2
    Lifetime Mode=::NewEnumerator1
    Sprite Size=<2 5>
    Sprite Size Mode=::NewEnumerator4
    UsePositionOffset=false
    Write Material Random=true
   )
   &ShapeLocation(MaskSpawnGroup=false, Override Seed=false, Sphere Radius=20, UseOwnerScale=true)
   &AddVelocity(
    Module.Use Falloff Distance=false
    Override Seed=false
    Velocity Mode=::NewEnumerator1
    Velocity Speed=&RandomRangeFloat(Minimum=85, Maximum=200)
   )
  ]
  particleUpdateScript [
   &ParticleState()
   &Drag(Drag=2, Use Linear Drag=true, Use Rotational Drag=true)
   &GravityForce(Gravity=<0 0 -300>)
   &SolveForcesAndVelocity(Clamp Velocity=false, Limit Acceleration=false)
   &ScaleSpriteSizeBySpeed(Max Scale Factor=<1.5 1>, Min Scale Factor=<0.7 1>)
   &DynamicMaterialParameters(
    Index 0 Param 2=&RandomRangeFloat(Minimum=0, Maximum=2)
    Module.Index0Float0Write=true
    Module.Index0Float1Write=true
    Module.Index0Float2Write=true
    Module.Index0Float3Write=true
   )
  ]
  rendererClasses [&NiagaraSpriteRendererProperties]
  renderers [{rendererClass=&NiagaraSpriteRendererProperties}]
 }
 SpikeStreaksParticle001 {
  emitterUpdateScript [
   &EmitterState()
   &SpawnBurst_Instantaneous(Spawn Count=5, Use Spawn Probability=false, UseLoopCountLimit=false)
  ]
  particleSpawnScript [
   &InitializeParticle(
    Color=&MultiplyLinearColor(
     Linear Color=<rgba 3 3 3 0.1>
     Scale Factor=&MakeVector4_FromLinearColor(LinearColor=$System.TODColorMultiplier)
    )
    Lifetime Max=0.4
    Lifetime Min=0.2
    Lifetime Mode=::NewEnumerator1
    Sprite Size=&Multiply_Vector2D_ByFloat(Float=&FN_FrontEndFloatModifier(Gameplay=1, Front End=0.6), Vector2D=<25 60>)
    Sprite Size Mode=::NewEnumerator4
    UsePositionOffset=false
    Write Material Random=true
   ) {module=InitializeParticle001}
   &ShapeLocation(
    MaskSpawnGroup=false
    Non Uniform Scale=<1 1 0>
    Ring / Disc Distribution Mode:ENiagaraSphereDistributionMode=::NewEnumerator3
    Ring Radius=5
    Rotation Angle=0.5
    Rotation Axis=<0 1 0>
    Rotation Mode=::NewEnumerator1
    Shape Primitive=::NewEnumerator4
    UseOwnerScale=true
   )
   &AddVelocity(Velocity=$StackContext.ShapeLocation.ShapeVector)
   &CameraOffset(Camera Offset Amount=25)
  ]
  particleUpdateScript [
   &ParticleState()
   &Drag(Drag=5, Use Linear Drag=true, Use Rotational Drag=true)
   &SolveForcesAndVelocity(Clamp Velocity=false, Limit Acceleration=false)
   &ScaleSpriteSize(
    Non-Uniform Curve Sprite Scale=DataInterface() {propertyValues {
      XCurve {
       defaultValue=1
       keys [
        {arriveTangent=4.1785, interpMode=RCIM_Cubic, leaveTangent=4.1785, tangentMode=RCTM_User, time=0, value=0}
        {arriveTangent=1.1179, interpMode=RCIM_Cubic, leaveTangent=1.1179, tangentMode=RCTM_User, time=0.2042, value=0.6228}
        {arriveTangent=0.0224, interpMode=RCIM_Cubic, leaveTangent=0.0224, tangentMode=RCTM_User, time=1, value=1}
       ]
      }
      YCurve {
       defaultValue=0
       keys [
        {arriveTangent=4.1785, interpMode=RCIM_Cubic, leaveTangent=4.1785, tangentMode=RCTM_User, time=0, value=0}
        {arriveTangent=1.1179, interpMode=RCIM_Cubic, leaveTangent=1.1179, tangentMode=RCTM_User, time=0.2042, value=0.6228}
        {arriveTangent=0.0224, interpMode=RCIM_Cubic, leaveTangent=0.0224, tangentMode=RCTM_User, time=1, value=1}
       ]
      }
     }}
    Scale Sprite Size Mode=::NewEnumerator3
   )
   &ScaleColor(
    Scale Alpha=&FloatFromCurve(FloatCurve=DataInterface() {propertyValues {Curve {keys [
         {interpMode=RCIM_Cubic, tangentMode=RCTM_User, time=0.5, value=1}
         {arriveTangent=-6.5352, interpMode=RCIM_Cubic, leaveTangent=-6.5352, tangentMode=RCTM_User, time=1, value=0}
        ]}}})
    Scale RGB=&Multiply_VectorByFloat(
     Vector=&MakeVectorFromLinearColorRGB(Color=&HueShiftLinearColor(
       Hue=-0.05
       Saturation=1
       Value=1.5
       Linear Color=&ColorFromCurve(ColorCurve=DataInterface() {propertyValues {
          AlphaCurve {keys []}
          BlueCurve {keys [<at 0 0.5434 RCIM_Cubic>,<at 0.9181 0.3 RCIM_Cubic>]}
          GreenCurve {keys [<at 0 0.0741 RCIM_Cubic>,<at 0.9181 0.0281 RCIM_Cubic>]}
          RedCurve {keys [<at 0 1.4583 RCIM_Cubic>,<at 0.9181 0.1414 RCIM_Cubic>]}
         }})
      ))
     Float=1
    )
    ScaleA=true
    ScaleRGB=true
   )
   &ScaleSpriteSize(Scale Sprite Size Mode=::NewEnumerator0) {module=ScaleSpriteSize001}
   /SetVariables(
    Particles.ConeEdge_Color:LinearColor=&Multiply_LinearColorByFloat(Linear Color=$Particles.Color, Float=1.5)
    Particles.ConeEdge_SpriteSize:Vector2f=&Multiply_Vector2D_ByFloat(Float=1.2, Vector2D=$Particles.SpriteSize)
   ) {isSetParameters=true}
  ]
  rendererClasses [&NiagaraSpriteRendererProperties]
  renderers [
   {rendererClass=&NiagaraSpriteRendererProperties}
   {rendererClass=&NiagaraSpriteRendererProperties, rendererIndex=1}
  ]
 }
 Streak {
  emitterUpdateScript [
   &EmitterState()
   &SpawnBurst_Instantaneous(Spawn Count=2, Use Spawn Probability=false, UseLoopCountLimit=false)
  ]
  particleSpawnScript [
   &InitializeParticle(
    Color Mode=::NewEnumerator0
    Lifetime=0.15
    Sprite Rotation Mode=::NewEnumerator2
    Sprite Size=<4 185>
    Sprite Size Mode=::NewEnumerator4
    UsePositionOffset=false
    Write Material Random=true
   )
   &SpriteFacingAndAlignment(WriteAlignment=false, WriteFacing=true)
  ]
  particleUpdateScript [
   &ParticleState()
   &SpriteRotationRate(Rotation Rate=&RandomRangeFloat(Minimum=-200, Maximum=200))
  ]
  rendererClasses [&NiagaraSpriteRendererProperties]
  renderers [{rendererClass=&NiagaraSpriteRendererProperties}]
 }
}
summary {
 systemName=NS_Emote_EvokeFind_KickBurst
 emitters [
  {emitterName=AudioWaves_Burst, rendererClasses [&NiagaraMeshRendererProperties]}
  {emitterName=SpikeStreaksParticle001, rendererClasses [&NiagaraSpriteRendererProperties]}
  {
   emitterName=Flash_Elements
   rendererClasses [&NiagaraSpriteRendererProperties, &NiagaraLightRendererProperties]
  }
  {emitterName=Minimal, rendererClasses [&NiagaraSpriteRendererProperties]}
  {emitterName=Sparks, rendererClasses [&NiagaraSpriteRendererProperties]}
  {emitterName=ParticleDecal, rendererClasses [&NiagaraSpriteRendererProperties, &NiagaraLightRendererProperties]}
  {emitterName=Streak, rendererClasses [&NiagaraSpriteRendererProperties]}
  {emitterName=CircleShape, rendererClasses [&NiagaraSpriteRendererProperties]}
 ]
 userVariables [
  User.bIsFrontEnd {type=NiagaraBool}
  User.bIsFrontEndPreview {type=NiagaraBool}
 ]
}
