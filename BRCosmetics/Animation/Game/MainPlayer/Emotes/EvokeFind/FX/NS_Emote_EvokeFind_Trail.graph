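# Niagara trail effect for the EvokeFind emote, spawned from sampled locations on a skeletal mesh.
# Two ribbon emitters trail the sampled points with animated color gradients and width falloff, setting ribbon facing from the sampled rotation.
# Hands spawns short-lived fading meshes dimmed at night by a time-of-day curve scale.
# ParticleDecal bursts a glowing sprite that re-samples the mesh location each update and drives a light from its color.
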NiagaraSystem /BRCosmetics/Animation/Game/MainPlayer/Emotes/EvokeFind/FX/NS_Emote_EvokeFind_Trail
dependencies {
 usedDataInterfaces [&NiagaraDataInterfaceColorCurve, &NiagaraDataInterfaceCurve, &NiagaraDataInterfaceSkeletalMesh]
 usedDynamicInputs []
 usedModules [
  &self:SystemSpawnScript.SystemScriptSource.SystemScriptGraph.NiagaraNodeAssignment_0.SetVariables_#1
  &SystemState
 ]
}
emitters {
 Hands {
  emitterUpdateScript [
   &EmitterState()
   &SpawnRate(SpawnRate=35, Use Spawn Probability=false)
  ]
  particleSpawnScript [
   &InitializeParticle(
    Color=<rgba 2 2 2 1>
    Lifetime=0.25
    Mesh Scale=<0.25 -0.25 0.2>
    Mesh Scale Mode=::NewEnumerator4
    UsePositionOffset=false
    Write Material Random=true
   )
   &InitialMeshOrientation(
    Mesh Axis To Orient=<0 0 -1>
    Mesh Orientation Mode=::NewEnumerator2
    Orientation Vector=<0 0 1>
    Rotation=<0.25 0 0>
    Use Rotation Vector=true
   )
   &SkeletalMeshLocation(Bone Sampling Mode=::NewEnumerator4, MaskSpawnGroup=false, Skeletal Mesh=$System.Skeletal Mesh)
  ]
  particleUpdateScript [
   &ParticleState()
   &ScaleColor(
    Scale Alpha=&FloatFromCurve(FloatCurve=DataInterface() {propertyValues {Curve {
        defaultValue=1
        keys [{interpMode=RCIM_Cubic, tangentMode=RCTM_User, time=0, value=1}, {interpMode=RCIM_Cubic, tangentMode=RCTM_User, time=1, value=0}]
       }}})
    Scale RGB=&VectorFromFloat(Value=&FloatFromCurve(
      FloatCurve=DataInterface() {propertyValues {Curve {keys [<at 0 50>,<at 0.6 5>,<at 1 0>]}}}
      Scale Curve=&FNTimeOfDayScalar(Night Scale=0.5, Morning Scale=1, Evening Scale=1)
     ))
    ScaleA=true
    ScaleRGB=true
   )
  ]
  rendererClasses [&NiagaraMeshRendererProperties]
  renderers [{rendererClass=&NiagaraMeshRendererProperties, rendererIndex=0}]
 }
 Minimal {
  emitterUpdateScript [
   &EmitterState()
   &SpawnRate(SpawnRate=40, Use Spawn Probability=false)
  ]
  particleSpawnScript [
   &InitializeParticle(Color=<rgba 3 3 3 1>, Lifetime=0.3, Ribbon Width=35, Ribbon Width Mode=::NewEnumerator1, UsePositionOffset=false, Write Material Random=true)
   &SkeletalMeshLocation(Bone Sampling Mode=::NewEnumerator4, MaskSpawnGroup=false, Skeletal Mesh=$System.Skeletal Mesh)
  ]
  particleUpdateScript [
   &ParticleState()
   &ScaleColor(
    Linear Color Curve=DataInterface() {propertyValues {
      AlphaCurve {keys [<at 0.0089 0>,<at 0.0755 1 RCIM_Cubic>,<at 0.9985 0>]}
      BlueCurve {keys [<at 0.1029 1 RCIM_Cubic>,<at 0.6178 1>,<at 0.9292 1 RCIM_Cubic>]}
      GreenCurve {keys [<at 0.1029 0.8879 RCIM_Cubic>,<at 0.6178 0.1812>,<at 0.9292 0.0953 RCIM_Cubic>]}
      RedCurve {keys [<at 0.1029 0.159 RCIM_Cubic>,<at 0.6178 0.0545>,<at 0.9292 0.9131 RCIM_Cubic>]}
     }}
    Scale Mode=::NewEnumerator2
   )
   &ScaleColor(
    Scale Alpha=&FloatFromCurve(FloatCurve=DataInterface() {propertyValues {Curve {keys [{arriveTangent=-2.2326, interpMode=RCIM_Cubic, leaveTangent=-2.2326, tangentMode=RCTM_User, time=0, value=1}, <at 0.4 0>]}}})
    Scale RGB=&Multiply_Vector(
     A=&MakeVectorFromLinearColorRGB(Color=$System.TODColorMultiplier)
     B=&Multiply_VectorByFloat(
      Vector=&MakeVectorFromLinearColorRGB(Color=&ColorFromCurve(ColorCurve=DataInterface() {propertyValues {
          AlphaCurve {keys [<at 0 1 RCIM_Cubic>]}
          BlueCurve {keys [
            <at 0.0813 1 RCIM_Cubic>
            <at 0.2728 1>
            {arriveTangent=-0.7194, interpMode=RCIM_Cubic, leaveTangent=-0.7194, time=0.6042, value=0.5576}
            <at 0.8468 0.5871>
           ]}
          GreenCurve {keys [
            <at 0.0813 0.8879 RCIM_Cubic>
            <at 0.2728 0.1812>
            {arriveTangent=-0.0643, interpMode=RCIM_Cubic, leaveTangent=-0.0643, time=0.6042, value=0.5871}
            <at 0.8468 0.1443>
           ]}
          RedCurve {keys [
            <at 0.0813 0.159 RCIM_Cubic>
            <at 0.2728 0.0545>
            {arriveTangent=-0.0274, interpMode=RCIM_Cubic, leaveTangent=-0.0274, time=0.6042, value=0.2599}
            <at 0.8468 0.0387>
           ]}
         }}))
      Float=2
     )
    )
    ScaleA=true
    ScaleRGB=true
   ) {enabled=false, module=ScaleColor001}
   /SetVariables(Particles.RibbonFacing=&RotateVectorByQuaternion(Vector To Rotate=<1 0 0>, Quaternion=$Particles.SkeletalMeshLocation.SampledRotation)) {isSetParameters=true}
   &ScaleRibbonWidth(Ribbon Width Scale=&FloatFromCurve(FloatCurve=DataInterface() {propertyValues {Curve {
        defaultValue=0
        keys [
         {tangentMode=RCTM_User, time=0, value=1}
         {arriveTangent=-0.0405, interpMode=RCIM_Cubic, leaveTangent=-0.0405, tangentMode=RCTM_User, time=0.7009, value=1}
         {arriveTangent=-8.0089, interpMode=RCIM_Cubic, leaveTangent=-8.0089, tangentMode=RCTM_User, time=1, value=0}
        ]
       }}}))
  ]
  rendererClasses [&NiagaraSpriteRendererProperties, &NiagaraRibbonRendererProperties]
  renderers [
   {rendererClass=&NiagaraSpriteRendererProperties}
   {rendererClass=&NiagaraRibbonRendererProperties, rendererIndex=1}
  ]
 }
 Minimal001 {
  emitterUpdateScript [
   &EmitterState()
   &SpawnRate(SpawnRate=40, Use Spawn Probability=false)
  ]
  particleSpawnScript [
   &InitializeParticle(
    Color=<rgba 1 1 1 0.2>
    Lifetime=0.3
    Ribbon Width=42
    Ribbon Width Mode=::NewEnumerator1
    UsePositionOffset=false
    Write Material Random=true
   )
   &SkeletalMeshLocation(Bone Sampling Mode=::NewEnumerator4, MaskSpawnGroup=false, Skeletal Mesh=$System.Skeletal Mesh)
  ]
  particleUpdateScript [
   &ParticleState()
   &ScaleColor(
    Linear Color Curve=DataInterface() {propertyValues {
      AlphaCurve {keys [<at 0.008 0>,<at 0.0929 1 RCIM_Cubic>,<at 0.9684 0>]}
      BlueCurve {keys [
        <at 0.0813 1 RCIM_Cubic>
        <at 0.2728 1>
        {arriveTangent=-0.7194, interpMode=RCIM_Cubic, leaveTangent=-0.7194, time=0.6042, value=0.5576}
        <at 0.8468 0.5871>
       ]}
      GreenCurve {keys [
        <at 0.0813 0.8879 RCIM_Cubic>
        <at 0.2728 0.1812>
        {arriveTangent=-0.0643, interpMode=RCIM_Cubic, leaveTangent=-0.0643, time=0.6042, value=0.5871}
        <at 0.8468 0.1443>
       ]}
      RedCurve {keys [
        <at 0.0813 0.159 RCIM_Cubic>
        <at 0.2728 0.0545>
        {arriveTangent=-0.0274, interpMode=RCIM_Cubic, leaveTangent=-0.0274, time=0.6042, value=0.2599}
        <at 0.8468 0.0387>
       ]}
     }}
    Scale Mode=::NewEnumerator2
   )
   /SetVariables(Particles.RibbonFacing=&RotateVectorByQuaternion(Vector To Rotate=<1 0 0>, Quaternion=$Particles.SkeletalMeshLocation.SampledRotation)) {isSetParameters=true}
   &ScaleRibbonWidth(Ribbon Width Scale=&FloatFromCurve(FloatCurve=DataInterface() {propertyValues {Curve {
        defaultValue=1
        keys [
         {arriveTangent=8.4448, interpMode=RCIM_Cubic, leaveTangent=-0.1941, tangentMode=RCTM_Break, time=0.2129, value=0.9766}
         {arriveTangent=-2.5932, interpMode=RCIM_Cubic, leaveTangent=-2.5932, tangentMode=RCTM_User, time=1, value=0}
        ]
       }}}))
  ]
  rendererClasses [&NiagaraSpriteRendererProperties, &NiagaraRibbonRendererProperties]
  renderers [
   {rendererClass=&NiagaraSpriteRendererProperties}
   {rendererClass=&NiagaraRibbonRendererProperties, rendererIndex=1}
  ]
 }
 ParticleDecal {
  emitterUpdateScript [
   &EmitterState()
   &SpawnBurst_Instantaneous(Use Spawn Probability=false, UseLoopCountLimit=false)
  ]
  particleSpawnScript [
   &InitializeParticle(
    Color=<rgba 4.5747 4.7676 5 1>
    Lifetime=0.6
    Mesh Scale Mode=::NewEnumerator3
    Mesh Uniform Scale=2
    Position Offset=<0 0 100>
    Sprite Size Mode=::NewEnumerator3
    Uniform Sprite Size=50
    UsePositionOffset=false
    Write Material Random=true
   )
   &CameraOffset(Camera Offset Amount=30)
   &DynamicMaterialParameters(
    Index 0 Param 1=1
    Index 0 Param 3=1
    Module.Index0Param0Write=true
    Module.Index0Param1Write=true
    Module.Index0Param2Write=true
    Module.Index0Param3Write=true
   )
  ]
  particleUpdateScript [
   &ParticleState()
   &SkeletalMeshLocation(Bone Sampling Mode=::NewEnumerator4, MaskSpawnGroup=false, Skeletal Mesh=$System.Skeletal Mesh)
   &ScaleColor(
    Linear Color Curve=DataInterface() {propertyValues {
      AlphaCurve {keys [<at 0.0755 1 RCIM_Cubic>,<at 0.9985 0>]}
      BlueCurve {keys [<at 0.1029 1 RCIM_Cubic>,<at 0.5859 1>,<at 0.9292 1 RCIM_Cubic>]}
      GreenCurve {keys [<at 0.1029 0.8879 RCIM_Cubic>,<at 0.5859 0.1812>,<at 0.9292 0.0953 RCIM_Cubic>]}
      RedCurve {keys [<at 0.1029 0.159 RCIM_Cubic>,<at 0.5859 0.0545>,<at 0.9292 0.9131 RCIM_Cubic>]}
     }}
    Scale Mode=::NewEnumerator2
   )
   &FN_Light_Attributes(
    Color=$Particles.Color
    Create Color Attribute=true
    Create Exponent Attribute=true
    Create Radius Attribute=true
    Exponent=4
    Radius=60
    Scale Color=2
   )
   /SetVariables(
    Particles.GlowColor=&ColorFromCurve(ColorCurve=DataInterface() {propertyValues {
       AlphaCurve {keys [<at 0.0418 0.3 RCIM_Cubic>,<at 0.9985 0>]}
       BlueCurve {keys [<at 0 1 RCIM_Cubic>,<at 0.274 1>,<at 0.7128 1 RCIM_Cubic>]}
       GreenCurve {keys [<at 0 0.8879 RCIM_Cubic>,<at 0.274 0.1812>,<at 0.7128 0.0953 RCIM_Cubic>]}
       RedCurve {keys [<at 0 0.159 RCIM_Cubic>,<at 0.274 0.0545>,<at 0.7128 0.9131 RCIM_Cubic>]}
      }})
    Particles.GlowSize=&Vector2DFromFloat(Value=30)
   ) {isSetParameters=true}
  ]
  rendererClasses [&NiagaraSpriteRendererProperties, &NiagaraLightRendererProperties]
  renderers [
   {rendererClass=&NiagaraSpriteRendererProperties}
   {rendererClass=&NiagaraSpriteRendererProperties, rendererIndex=1}
   {rendererClass=&NiagaraLightRendererProperties, rendererIndex=2}
  ]
 }
}
summary {
 systemName=NS_Emote_EvokeFind_Trail
 emitters [
  {emitterName=Minimal, rendererClasses [&NiagaraSpriteRendererProperties, &NiagaraRibbonRendererProperties]}
  {emitterName=Minimal001, rendererClasses [&NiagaraSpriteRendererProperties, &NiagaraRibbonRendererProperties]}
  {emitterName=ParticleDecal, rendererClasses [&NiagaraSpriteRendererProperties, &NiagaraLightRendererProperties]}
  {emitterName=Hands, rendererClasses [&NiagaraMeshRendererProperties]}
 ]
 userVariables [
  User.bIsFrontEnd {type=NiagaraBool}
  User.bIsFrontEndPreview {type=NiagaraBool}
  User.RibbonIDOffset {type=NiagaraInt32}
 ]
}
